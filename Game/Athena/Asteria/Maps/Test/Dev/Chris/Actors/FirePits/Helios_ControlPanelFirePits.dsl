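; Interactable Helios control panel that triggers fire pit traps, firing pattern indices on fire pit managers listed in ActivationData.
; Interaction sets ButtonInteracted, waits a pre-activation delay, fires each entry, then resets after a cooldown.
; Authority disables panels without activation data via NoDataDisable; RepNotifies drive button glow, plus alarm and click sounds on press.
; Clients fade the temporary alarm shortly after use or on death, which also stops console beeps.
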(event EventOnDeathPlayEffects
  (params (Damage :type Float%20%28single-precision%29) (DamageTags :type Gameplay%20Tag%20Container%20Structure%20%28by%20ref%29) (Momentum :type Vector :default "0, 0, 0") (HitInfo :type Hit%20Result%20Structure%20%28by%20ref%29) (InstigatedBy :type Fort%20Pawn%20Object%20Reference) (DamageCauser :type Actor%20Object%20Reference) (EffectContext :type Gameplay%20Effect%20Context%20Handle%20Structure))
  (bind _spot_oilrig_controlconsole_beeps_loop_cue (Variables|Default|GetSpot_OilRig_ControlConsole_Beeps_Loop_Cue))
  ;[1 1 1] On Death Stop Beeping Audio
  (Utilities|IsValid _spot_oilrig_controlconsole_beeps_loop_cue
    (:Is%20Valid
      (if (Utilities|IsValid _spot_oilrig_controlconsole_beeps_loop_cue :!class /Script/Engine.KismetSystemLibrary)
        (Audio|Components|Audio|FadeOut _spot_oilrig_controlconsole_beeps_loop_cue 0.2 0.0 "Linear" :!class /Script/Engine.AudioComponent)))))

(event EventBeginPlay
  ;[1 1 1] Init
  (Networking|SwitchHasAuthority
    (:Authority
      (if (>= (Utilities|Array|Length (Variables|Default|GetActivationData)) 1)
        (Variables|Default|SetCooldownDuration (Math|Float|Clamp(Float) (Variables|Default|GetCooldownDuration) 3.5 99999.0))
        (EventDispatchers|BindEventtoOnDied (delegate Custom|OnDied_Event))
        (else
          (Variables|Default|SetNoDataDisable true))))))

(event Custom|OnDied_Event
  (params (DamagedActor :type Actor%20Object%20Reference) (Damage :type Float%20%28single-precision%29) (InstigatedBy :type Controller%20Object%20Reference) (DamageCauser :type Actor%20Object%20Reference) (HitLocation :type Vector) (FHitComponent :type Primitive%20Component%20Object%20Reference) (BoneName :type Name) (Momentum :type Vector))
  ;[1 1 1] Init
  (CallFunction|CleanupTempAudio))

(event Custom|CleanupTempAudio
  (bind _alarmsoundref (Variables|Default|GetAlarmSoundRef))
  ;[1 1 1] Init
  (Networking|SwitchHasAuthority
    (:Remote
      (Utilities|IsValid _alarmsoundref
        (:Is%20Valid
          (Audio|Components|Audio|FadeOut _alarmsoundref 1.0 0.0 "Linear" :!class /Script/Engine.AudioComponent))))))

(event Interaction|EventBlueprintOnInteract
  (params (InteractingPawn :type Fort%20Pawn%20Object%20Reference) (InteractionBeingAttempted :type EInteractionBeingAttempted%20Enum :default "FirstInteraction"))
  ;[1 1 1] Interact
  (Variables|Default|SetButtonInteracted true)
  (Utilities|FlowControl|Delay (Variables|Default|GetPreActivationDelay))
  (Utilities|FlowControl|Sequence
    (:then_0
      (for _array_element (Variables|Default|GetActivationData)
        (Class|HeliosFirePitManager|FireAllatPatternIndex (Utilities|Struct|BreakHeliosCPStruct _array_element) (.PatternIndexToFire_6_01E7F07349389119F032CB86205CB988 (Utilities|Struct|BreakHeliosCPStruct _array_element))))
      (Utilities|FlowControl|Delay (Variables|Default|GetCooldownDuration))
      (Variables|Default|SetButtonInteracted false))
    (:then_1
      (Utilities|FlowControl|Delay 2.0)
      (CallFunction|FadeOutTempAudio))))

(event Custom|FadeOutTempAudio
  (bind _alarmsoundref (Variables|Default|GetAlarmSoundRef))
  ;[1 0.2552 0.3929] Temp Audio hack
  (Networking|SwitchHasAuthority
    (:Remote
      (Utilities|IsValid _alarmsoundref
        (:Is%20Valid
          (Audio|Components|Audio|FadeOut _alarmsoundref 1.5 0.0 "Linear" :!class /Script/Engine.AudioComponent))))))

(fn BlueprintCanInteract
  (params (InteractingPawn :type Fort%20Pawn%20Object%20Reference) (InteractionBeingAttempted :type EInteractionBeingAttempted%20Enum :default "FirstInteraction") (InteractionType :type TInteractionType%20Enum :default "IT_NoInteraction"))
  (return (!= (Variables|Default|GetButtonInteracted) (!= (Variables|Default|GetNoDataDisable) true))))

(fn BlueprintGetInteractionString
  (params (InteractingPawn :type Fort%20Pawn%20Object%20Reference) (InteractionBeingAttempted :type EInteractionBeingAttempted%20Enum :default "FirstInteraction"))
  (return "NSLOCTEXT(\"[9949FD51BBB9E8ECFB36F24FE4C1D256]\", \"488475A04CF85AAE1167FB8E0D74E517\", \"Trigger Fire Traps\")"))

(fn OnRep_ButtonInteracted
  (if (Variables|Default|GetButtonInteracted)
    (bind _returnvalue (Rendering|Material|CreateDynamicMaterialInstance (Variables|Default|Getbuttonmesh) 0 "/Game/Athena/Asteria/Maps/Test/Dev/Chris/Actors/FirePits/M_ButtonGlowTest.M_ButtonGlowTest" "None" :!class /Script/Engine.PrimitiveComponent))
    (Rendering|Material|SetScalarParameterValue _returnvalue "pulse" 0.0 :!class /Script/Engine.MaterialInstanceDynamic)
    (bind _returnvalue_2 (Audio|SpawnSoundAttached :Sound "/Game/Sounds/Asteria/POI/HeistPOIs/Alarm_Lockdown_LP_01_Cue.Alarm_Lockdown_LP_01_Cue" :AttachToComponent (Variables|Default|Getbuttonmesh) :AttachPointName "None" :Location (Transformation|GetActorLocation self) :LocationType "KeepWorldPosition" :bStopWhenAttachedToDestroyed false :VolumeMultiplier 1.0 :PitchMultiplier 1.0 :StartTime 0.0 :bAutoDestroy true))
    (bind _output_get (Variables|Default|SetAlarmSoundRef _returnvalue_2))
    (Audio|Components|Audio|Play _output_get 0.0)
    (bind _returnvalue_3 (Audio|SpawnSoundAttached :Sound "/Game/Sounds/Creative/Toys/Locks/LockDevice_UpdateItemSwitch_Cue.LockDevice_UpdateItemSwitch_Cue" :AttachToComponent (Variables|Default|Getbuttonmesh) :AttachPointName "None" :Location (Transformation|GetActorLocation self) :LocationType "KeepWorldPosition" :bStopWhenAttachedToDestroyed false :VolumeMultiplier 1.0 :PitchMultiplier 1.0 :StartTime 0.0 :bAutoDestroy true))
    (Audio|Components|Audio|Play _returnvalue_3 0.0)
    (else
      (bind _returnvalue_4 (Rendering|Material|CreateDynamicMaterialInstance (Variables|Default|Getbuttonmesh) 0 "/Game/Athena/Asteria/Maps/Test/Dev/Chris/Actors/FirePits/M_ButtonGlowTest.M_ButtonGlowTest" "None" :!class /Script/Engine.PrimitiveComponent))
      (Rendering|Material|SetScalarParameterValue _returnvalue_4 "pulse" 1.0 :!class /Script/Engine.MaterialInstanceDynamic))))

(fn OnRep_NoDataDisable
  (if (Variables|Default|GetNoDataDisable)
    (bind _returnvalue (Rendering|Material|CreateDynamicMaterialInstance (Variables|Default|Getbuttonmesh) 0 "/Game/Athena/Asteria/Maps/Test/Dev/Chris/Actors/FirePits/M_ButtonGlowTest.M_ButtonGlowTest" "None" :!class /Script/Engine.PrimitiveComponent))
    (Rendering|Material|SetScalarParameterValue _returnvalue "pulse" 0.0 :!class /Script/Engine.MaterialInstanceDynamic)))

(fn ConstructionScript)
