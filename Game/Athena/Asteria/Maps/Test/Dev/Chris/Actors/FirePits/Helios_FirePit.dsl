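; Helios fire pit actor that erupts as a timed geyser when fired, launching and damaging overlapping players and vehicles.
; Firing briefly clears isDisabled, whose OnRep toggles cone collision and burst effects.
; Overlap also impulses physics objects, inverts gravity on projectiles, C4 and toys, and capsule-traces to damage buildings along launch paths; exit grants fall-damage immunity.
; BeginPlay randomizes cone materials and binds attachment deaths to destroy the pit.
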(event EventBeginPlay
  ;[0.05098 1 0.01223] Begin Play
  (Variables|Default|SetisDisabled true)
  (Utilities|FlowControl|Sequence
    (:then_0
      (CallFunction|SetupMaterialParams)
      (Variables|Default|SetSavedConeScale (Variables|/Script/Engine.SceneComponent|GetRelativeScale3D (Variables|Default|GetS_FX_GeyserCone)))
      (for _array_element (Variables|FirePit|GetAttachmentActors)
        (EventDispatchers|BindEventtoOnDied _array_element (delegate Custom|DestroyPit))))))

(event Custom|DestroyPit
  (params (DamagedActor :type Actor%20Object%20Reference) (Damage :type Float%20%28single-precision%29) (InstigatedBy :type Controller%20Object%20Reference) (DamageCauser :type Actor%20Object%20Reference) (HitLocation :type Vector) (FHitComponent :type Primitive%20Component%20Object%20Reference) (BoneName :type Name) (Momentum :type Vector))
  (bind _self self)
  ;[0.05098 1 0.01223] Begin Play
  (Utilities|FlowControl|DoOnce false
    (:Completed
      (CallFunction|CleanupAnyOIverlappingonDeath)
      (Utilities|IsValid _self
        (:Is%20Valid
          (Actor|DestroyActor _self))))))

(event Custom|SetupMaterialParams
  (bind _s_fx_geysercone (Variables|Default|GetS_FX_GeyserCone))
  ;[0.008445 0.4764 1] Set material values
  (Rendering|Material|SetScalarParameterValueonMaterials _s_fx_geysercone "UVScale" (Math|Random|RandomFloatinRange 2.0 4.0))
  (Rendering|Material|SetScalarParameterValueonMaterials _s_fx_geysercone "SlantAmount" (Math|Random|RandomFloatinRange -5e+02 5e+02))
  (Rendering|Material|SetScalarParameterValueonMaterials _s_fx_geysercone "WPO_Amount" (Math|Random|RandomFloatinRange 0.0 5e+02))
  (Rendering|Material|SetScalarParameterValueonMaterials _s_fx_geysercone "WorldTexUVScale" (Math|Random|RandomFloatinRange 0.5 2.0))
  (bind _output_get (Variables|Default|SetGeyserContrast (Math|Random|RandomFloatinRange 1.0 3.0)))
  (Rendering|Material|SetScalarParameterValueonMaterials _s_fx_geysercone "Top Edge Contrast" _output_get))

(event Custom|FireTheFirePit
  ;[1 0 0.9359] fire pits from manager or player button
  (Networking|SwitchHasAuthority
    (:Authority
      (if (not (Variables|Default|GetPitIsActive))
        (Variables|Default|SetPitIsActive true)
        (Variables|Default|SetisDisabled false)
        (bind _returnvalue (Utilities|Time|SetTimerbyEvent (delegate Custom|ResetGeyserWait) 1.25))
        (Variables|Default|SetActiveTimerHandle _returnvalue)))))

(event OnComponentBeginOverlap(S_FX_GeyserCone)
  (params (OverlappedComponent :type Primitive%20Component%20Object%20Reference) (OtherActor :type Actor%20Object%20Reference) (OtherComp :type Primitive%20Component%20Object%20Reference) (OtherBodyIndex :type Integer) (bFromSweep :type Boolean) (SweepResult :type Hit%20Result%20Structure%20%28by%20ref%29))
  (bind _sf_launchvelocity (Variables|Default|GetSF_LaunchVelocity))
  (bind _returnvalue (ScalableFloat|GetValueatLevel _sf_launchvelocity))
  (bind _returnvalue_2 (Transformation|GetActorUpVector))
  (bind _returnvalue_3 (* _returnvalue_2 _returnvalue))
  (bind _returnvalue_4 (Pawn|GetMovementBaseActor 0))
  (bind _returnvalue_5 (Transformation|GetActorUpVector))
  (bind _sf_vehiclelaunchvelocity (Variables|Default|GetSF_VehicleLaunchVelocity))
  (bind _returnvalue_6 (ScalableFloat|GetValueatLevel _sf_vehiclelaunchvelocity))
  (bind _returnvalue_7 (* _returnvalue_5 _returnvalue_6))
  (bind (_returnvalue_x _returnvalue_y _returnvalue_z) (Math|Vector|BreakVector (Transformation|GetActorLocation)))
  ;[1 0.3378 0] Launch players and vehicles if they overlap the geyser while it's active. Handle projectiles and toys/C4 separately.
  (if (Variables|Default|GetShouldBoost)
    (bind _asfort_pawn (Utilities|Casting|CastToFortPawn OtherActor)
      (:then
        (Utilities|IsValid _returnvalue_4
          (:Is%20Valid
            (if (Utilities|IsValid (Actor|GetComponentbyClass _returnvalue_4 "/Script/FortniteGame.FortPhysicsObjectComponent") :!class /Script/Engine.KismetSystemLibrary)
              (bind _output_get (Variables|Default|SetPhysicsObjectToLaunch _returnvalue_4))
              (Variables|Default|SetLaunchedActor _output_get)))
          (:Is%20Not%20Valid
            (if (not (or (GameplayTags|TagContainer|CompareTagAssetInterfacetoOtherTagContainers (Variables|Default|GetHawaii) (Variables|Default|GetRiding) _asfort_pawn) (.Case_1%20True (GameplayTags|TagContainer|CompareTagAssetInterfacetoOtherTagContainers (Variables|Default|GetHawaii) (Variables|Default|GetRiding) _asfort_pawn))))
              (Networking|SwitchHasAuthority
                (:Authority
                  (Variables|Default|SetLaunchedActor _asfort_pawn)
                  (Ability|SendGameplayEventtoActor _asfort_pawn "(TagName=\"Abilities.Generic.Vent.Enter\")" (Utilities|Struct|MakeGameplayEventData :Target self :ContextHandle "()" :InstigatorTags "(GameplayTags=)" :TargetTags "(GameplayTags=)" :EventMagnitude (Variables|Default|GetUpwardLaunchVelocity) :TargetData "()"))
                  (Environment|ApplyGlobalEnvironmentDamagetoActor _asfort_pawn (Variables|FirePit|GetFirePitPlayerDamage))
                  (Character|LaunchCharacter _asfort_pawn _returnvalue_3 (Variables|Default|GetXYOverride) (Variables|Default|GetZOverride))
                  ;;[0.6434 0.2072 1] Use Geysers in Different Matches Quest hook in
                  (Utilities|IsValid (Fortnite|GetFortPlayerControllerfromActor _asfort_pawn)
                    (:Is%20Valid
                      (Utilities|IsValid (Quests|GetQuestManager (Fortnite|GetFortPlayerControllerfromActor _asfort_pawn) "Invalid")
                        (:Is%20Valid
                          (Quest|SendComplexCustomStatEvent (Quests|GetQuestManager (Fortnite|GetFortPlayerControllerfromActor _asfort_pawn) "Invalid") 0 "(GameplayTags=((TagName=\"Athena.Quests.Guyser\")))" "(GameplayTags=)" 1))))
                    (:Is%20Not%20Valid
                      _))
                  (CallFunction|DestroyBuildingProps _asfort_pawn (Ability|GetAbilitySystemComponent _asfort_pawn) _returnvalue_3))
                (:Remote
                  (bind (_returnvalue_x_2 _returnvalue_y_2 _returnvalue_z_2) (Math|Vector|BreakVector (Transformation|GetActorLocation _asfort_pawn)))
                  (Projectile|AddGameplayCue%28Looping%29Local _asfort_pawn (Variables|Default|GetVentCue) (Ability|GameplayCue|MakeGameplayCueParameters :NormalizedMagnitude (Math|Conversions|ToFloat(Boolean) (<= _returnvalue_z_2 (+ _returnvalue_z 8343.651367))) :RawMagnitude 0.0 :EffectContext "()" :AggregatedSourceTags "(GameplayTags=)" :AggregatedTargetTags "(GameplayTags=)" :Location "0,0,0" :Normal "0,0,0" :GameplayEffectLevel 1 :AbilityLevel 1 :bReplicateLocationWhenUsingMinimalRepProxy false))
                  (ForceFeedback|SpawnForceFeedbackAttached "/Game/Weapons/ForceFeedback/Explosions/FF_Generic_ExplosionLarge_Near.FF_Generic_ExplosionLarge_Near" (Variables|/Script/Engine.Actor|GetRootComponent _asfort_pawn))))))) ; Do not launch players if they are riding on an object we will launch
      )
      (:CastFailed
        (bind _asfort_vehicle_interface (Utilities|Casting|CastToFortVehicleInterface OtherActor)
          (:then
            (CallFunction|OnVehicleEnterZone (Vehicle|GetActor _asfort_vehicle_interface) -0.6)
            (Vehicle|SetIgnoreAllFallingDamage _asfort_vehicle_interface true)
            (Variables|Default|SetLaunchedActor (Vehicle|GetActor _asfort_vehicle_interface))
            (Audio|BroadcastAbilityImpactatLocation 0 (Vehicle|GetActor _asfort_vehicle_interface) (+ (Vehicle|GetVehicleVelocity _asfort_vehicle_interface) _returnvalue_7) 777.0)
            (Environment|ApplyGlobalEnvironmentDamagetoActor (Vehicle|GetActor _asfort_vehicle_interface) (Variables|FirePit|GetFirePitVehicleDamage))
            (CallFunction|DestroyBuildingProps (Vehicle|GetActor _asfort_vehicle_interface) (Ability|GetAbilitySystemComponent (Vehicle|GetActor _asfort_vehicle_interface)) (+ (Vehicle|GetVehicleVelocity _asfort_vehicle_interface) _returnvalue_7)))
          (:CastFailed
            (bind _asfort_projectile_base (Utilities|Casting|CastToFortProjectileBase OtherActor)
              (:then
                (Variables|Default|SetLaunchedActor _asfort_projectile_base)
                (CallFunction|ProjEnter _asfort_projectile_base))
              (:CastFailed
                (bind _asb_bga_athena_c4 (Utilities|Casting|CastToB_BGA_Athena_C4 OtherActor)
                  (:then
                    (Variables|/Script/Engine.ProjectileMovementComponent|SetProjectileGravityScale (* (Variables|/Script/Engine.ProjectileMovementComponent|GetProjectileGravityScale (Variables|/Game/Athena/Items/Consumables/C4/B_BGA_Athena_C4.B_BGA_Athena_C4_C|GetProjectileMovement1 _asb_bga_athena_c4)) -1.0) (Variables|/Game/Athena/Items/Consumables/C4/B_BGA_Athena_C4.B_BGA_Athena_C4_C|GetProjectileMovement1 _asb_bga_athena_c4))
                    (Variables|Default|SetLaunchedActor _asb_bga_athena_c4))
                  (:CastFailed
                    (bind _asphysics_ball_master (Utilities|Casting|CastToPhysicsBall_Master OtherActor)
                      (:then
                        (Variables|Default|SetLaunchedActor _asphysics_ball_master)
                        (CallFunction|ToyEnter _asphysics_ball_master))
                      (:CastFailed
                        (if (Utilities|IsValid (Actor|GetComponentbyClass OtherActor "/Script/FortniteGame.FortPhysicsObjectComponent") :!class /Script/Engine.KismetSystemLibrary)
                          (Variables|Default|SetLaunchedActor OtherActor)
                          (Variables|Default|SetPhysicsObjectToLaunch OtherActor))))))))))))
    (FortPhysicsSimulationLibrary|AddPhysicsImpulse (Variables|Default|GetPhysicsObjectToLaunch) (* (Transformation|GetActorUpVector) (ScalableFloat|GetValueatLevel (Variables|Default|GetSF_BoulderLaunchVelocity))) true "None" false false 0 "(MaxSimSizeAllowed=Large,bRejectEmbeddedInLandscape=False)")))

(event Custom|ResetGeyserWait
  ;[1 0 0.9359] fire pits from manager or player button
  (Variables|Default|SetisDisabled true)
  (Variables|Default|SetPitIsActive false))

(event Custom|ResetFirePitsTheFireAll
  ;[1 0 0.9359] fire pits from manager or player button
  ;;[1 1 1] called when a player fires all pits manually - stops currently active ones and resets them then fires
  (Utilities|Time|ClearandInvalidateTimerbyHandle (Variables|Default|GetActiveTimerHandle))
  (Variables|Default|SetisDisabled true)
  (Variables|Default|SetPitIsActive false)
  (CallFunction|FiretheFirePit))

(event OnComponentEndOverlap(S_FX_GeyserCone)
  (params (OverlappedComponent :type Primitive%20Component%20Object%20Reference) (OtherActor :type Actor%20Object%20Reference) (OtherComp :type Primitive%20Component%20Object%20Reference) (OtherBodyIndex :type Integer))
  (bind _returnvalue (Actor|GetComponentbyClass OtherActor "/Script/FortniteGame.FortPhysicsObjectComponent"))
  ;[0.1134 0 1] Set the launched targets to be immune to fall damage.
  (if (Variables|Default|GetShouldBoost)
    (bind _asfort_pawn (Utilities|Casting|CastToFortPawn OtherActor)
      (:then
        (Ability|SendGameplayEventtoActor _asfort_pawn "(TagName=\"Abilities.Generic.Vent.Exit\")")
        (Variables|/Script/FortniteGame.FortPawn|SetbIgnoreNextFallingDamage true _asfort_pawn)
        (Projectile|RemoveGameplayCue%28Looping%29Local _asfort_pawn (Variables|Default|GetVentCue) (Ability|GameplayCue|MakeGameplayCueParameters :NormalizedMagnitude 0.0 :RawMagnitude 0.0 :EffectContext "()" :AggregatedSourceTags "(GameplayTags=)" :AggregatedTargetTags "(GameplayTags=)" :Location "0,0,0" :Normal "0,0,0" :GameplayEffectLevel 1 :AbilityLevel 1 :bReplicateLocationWhenUsingMinimalRepProxy false)))
      (:CastFailed
        (bind _asfort_vehicle_interface (Utilities|Casting|CastToFortVehicleInterface OtherActor)
          (:then
            (CallFunction|OnVehicleExitZone (Vehicle|GetActor _asfort_vehicle_interface))
            (Vehicle|SetIgnoreAllFallingDamage _asfort_vehicle_interface false)
            (Vehicle|SetIgnoreNextFallingDamage _asfort_vehicle_interface true))
          (:CastFailed
            (bind _asfort_projectile_base (Utilities|Casting|CastToFortProjectileBase OtherActor)
              (:then
                (CallFunction|ProjExit _asfort_projectile_base))
              (:CastFailed
                (bind _asb_bga_athena_c4 (Utilities|Casting|CastToB_BGA_Athena_C4 OtherActor)
                  (:then
                    (Variables|/Script/Engine.ProjectileMovementComponent|SetProjectileGravityScale (* (Variables|/Script/Engine.ProjectileMovementComponent|GetProjectileGravityScale (Variables|/Game/Athena/Items/Consumables/C4/B_BGA_Athena_C4.B_BGA_Athena_C4_C|GetProjectileMovement1 _asb_bga_athena_c4)) -1.0) (Variables|/Game/Athena/Items/Consumables/C4/B_BGA_Athena_C4.B_BGA_Athena_C4_C|GetProjectileMovement1 _asb_bga_athena_c4)))
                  (:CastFailed
                    (bind _asphysics_ball_master (Utilities|Casting|CastToPhysicsBall_Master OtherActor))
                    (CallFunction|ToyExit _asphysics_ball_master)
                    (if (Utilities|IsValid _returnvalue :!class /Script/Engine.KismetSystemLibrary)
                      (PhysicsObject|DecrementNeverSleep _returnvalue))))))))))))

(fn CleanupAnyOIverlappingOnDeath
  (for _array_element (Collision|GetOverlappingActors (Variables|Default|GetS_FX_GeyserCone) "/Script/FortniteGame.FortPawn" :!class /Script/Engine.PrimitiveComponent)
    (bind _asfort_pawn (Utilities|Casting|CastToFortPawn _array_element))
    (Ability|SendGameplayEventtoActor _asfort_pawn "(TagName=\"Abilities.Generic.Vent.Exit\")")
    (Variables|/Script/FortniteGame.FortPawn|SetbIgnoreNextFallingDamage true _asfort_pawn)
    (Projectile|RemoveGameplayCue%28Looping%29Local _asfort_pawn (Variables|Default|GetVentCue) (Ability|GameplayCue|MakeGameplayCueParameters :NormalizedMagnitude 0.0 :RawMagnitude 0.0 :EffectContext "()" :AggregatedSourceTags "(GameplayTags=)" :AggregatedTargetTags "(GameplayTags=)" :Location "0,0,0" :Normal "0,0,0" :GameplayEffectLevel 1 :AbilityLevel 1 :bReplicateLocationWhenUsingMinimalRepProxy false))))

(fn DestroyBuildingProps
  (params (Actor :type Actor%20Object%20Reference) (AbilitySystemComp :type Ability%20System%20Component%20Object%20Reference) (LaunchVelocity :type Vector))
  (if (> (ScalableFloat|GetValueatLevel (Variables|Default|GetShouldDestroy)) 0.0)
    (bind _outhits (Collision|MultiCapsuleTraceForObjects :Start (Transformation|GetActorLocation Actor) :End (+ (Transformation|GetActorLocation Actor) (* (Math|Vector|Normalize LaunchVelocity) (ScalableFloat|GetValueatLevel (Variables|Default|GetDestroyDistance)))) :Radius 8e+01 :HalfHeight 1.5e+02 :ObjectTypes (Variables|Default|GetDestroyObjectTypes) :bTraceComplex false :DrawDebugType "None" :bIgnoreSelf true :TraceColor "(R=1.000000,G=0.000000,B=0.000000,A=1.000000)" :TraceHitColor "(R=0.000000,G=1.000000,B=0.000000,A=1.000000)" :DrawTime 8.0))
    (for _array_element _outhits
      (if (Variables|/Script/Engine.Actor|GetbCanBeDamaged (.HitActor (Collision|BreakHitResult _array_element)))
        (bind _bdidactorhavetags (Class|AthenaQuestFunctionLibrary|AthenaCheckActorHasTags (.HitActor (Collision|BreakHitResult _array_element)) "(GameplayTags=((TagName=\"Deimos.Spawner\")))"))
        (if (not _bdidactorhavetags)
          (Utilities|Casting|CastToBuildingTrap (.HitActor (Collision|BreakHitResult _array_element))
            (:CastFailed
              (bind _asbuilding_actor (Utilities|Casting|CastToBuildingActor (.HitActor (Collision|BreakHitResult _array_element))))
              (if (not (Utilities|Array|ContainsItem (Variables|FirePit|GetAttachmentActors) _asbuilding_actor))
                (GameplayEffects|ApplyGameplayEffectToTarget :self AbilitySystemComp :GameplayEffectClass "/Game/Athena/Items/Consumables/ShockwaveGrenade/GE_ShockGrenade_Damage.GE_ShockGrenade_Damage_C" :Target (Ability|GetAbilitySystemComponent 0) :Level 0.0)))))))))

(fn GetCollisionStaticMesh
  (return "0, 0, 0" "/Game/Athena/Environments/Island/Runes/Mesh/SM_CorruptedVent_A.SM_CorruptedVent_A"))

(fn GetMeshComponents
  (return 0))

(fn GeyserLaunch
  (GameplayCue|ExecuteGameplayCueOnActor(Burst) self (Variables|Default|GetLaunchGC) (Ability|GameplayCue|MakeGameplayCueParameters :NormalizedMagnitude 0.0 :RawMagnitude 0.0 :Location "0,0,0" :Normal "0,0,0" :Instigator (Variables|Default|GetLaunchedActor) :GameplayEffectLevel 1 :AbilityLevel 1 :bReplicateLocationWhenUsingMinimalRepProxy false))
  (Variables|Default|SetLaunchedActor 0) ; Clearing so the next time an ambient launch comes around we don't have this cached
)

(fn IgnoreWorldGridScaleOverride
  (return false))

(fn OnRep_IsActive
  (Networking|SwitchHasAuthority
    (:Remote
      (if (Variables|Default|GetIsActive)
        _))))

(fn OnRep_isDisabled
  (bind _geyserburstref (Variables|Default|GetGeyserBurstRef))
  (if (Variables|Default|GetisDisabled)
    (Collision|SetCollisionEnabled (Variables|Default|GetS_FX_GeyserCone) "NoCollision")
    (Variables|Default|SetLaunchedActor 0) ; Clearing so the next time an ambient launch comes around we don't have this cached
    (Utilities|IsValid _geyserburstref
      (:Is%20Valid
        (Components|Activation|Deactivate _geyserburstref)))
    (else
      (Collision|SetCollisionEnabled (Variables|Default|GetS_FX_GeyserCone) "QueryAndPhysics")
      (if (Variables|Default|GetAllowEffects)
        (bind _returnvalue (Effects|SpawnEmitteratLocation :EmitterTemplate "/Game/Effects/Fort_Effects/Gameplay/Ostrich/P_Ostrich_Explosion_Buildup.P_Ostrich_Explosion_Buildup" :Location (Transformation|GetWorldLocation (Variables|Default|GetIdleFX)) :Scale "1.250000,1.250000,1.250000" :bAutoDestroy true :PoolingMethod "None" :bAutoActivateSystem true))
        (Variables|Default|SetGeyserBurstRef _returnvalue)
        (GameplayCue|ExecuteGameplayCueOnActor(Burst) self (Variables|Default|GetLaunchGC) (Ability|GameplayCue|MakeGameplayCueParameters :NormalizedMagnitude 0.0 :RawMagnitude 0.0 :Location "0,0,0" :Normal "0,0,0" :Instigator (Variables|Default|GetLaunchedActor) :GameplayEffectLevel 1 :AbilityLevel 1 :bReplicateLocationWhenUsingMinimalRepProxy false))
        (Variables|Default|SetLaunchedActor 0) ; Clearing so the next time an ambient launch comes around we don't have this cached
      ))))

(fn OnRep_PrimeMode)

(fn ProjEnter
  (params (Projectile :type Fort%20Projectile%20Base%20Object%20Reference))
  (bind _projectilemovementcomponent (Variables|/Script/FortniteGame.FortProjectileBase|GetProjectileMovementComponent Projectile))
  (bind _projectilegravityscale (Variables|/Script/Engine.ProjectileMovementComponent|GetProjectileGravityScale _projectilemovementcomponent))
  (if (> _projectilegravityscale 0.0)
    (Variables|/Script/Engine.ProjectileMovementComponent|SetProjectileGravityScale (* _projectilegravityscale -1.0) _projectilemovementcomponent)
    (else
      (Variables|/Script/Engine.ProjectileMovementComponent|SetProjectileGravityScale (+ _projectilegravityscale (Variables|Default|GetZeroGravPrj)) _projectilemovementcomponent))))

(fn ProjExit
  (params (Projectile :type Fort%20Projectile%20Base%20Object%20Reference))
  (bind _projectilemovementcomponent (Variables|/Script/FortniteGame.FortProjectileBase|GetProjectileMovementComponent Projectile))
  (bind _projectilegravityscale (Variables|/Script/Engine.ProjectileMovementComponent|GetProjectileGravityScale _projectilemovementcomponent))
  (if (== _projectilegravityscale (Variables|Default|GetZeroGravPrj))
    (Variables|/Script/Engine.ProjectileMovementComponent|SetProjectileGravityScale (- _projectilegravityscale (Variables|Default|GetZeroGravPrj)) _projectilemovementcomponent)
    (else
      (Variables|/Script/Engine.ProjectileMovementComponent|SetProjectileGravityScale (* _projectilegravityscale -1.0) _projectilemovementcomponent))))

(fn ToyEnter
  (params (Projectile :type Physics%20Ball%20Master%20Object%20Reference))
  (bind _projectilemovement (Variables|/Game/Building/ActorBlueprints/Prop/PhysicsBall_Master.PhysicsBall_Master_C|GetProjectileMovement Projectile))
  (bind _projectilegravityscale (Variables|/Script/Engine.ProjectileMovementComponent|GetProjectileGravityScale _projectilemovement))
  (if (> _projectilegravityscale 0.0)
    (Variables|/Script/Engine.ProjectileMovementComponent|SetProjectileGravityScale (* _projectilegravityscale -0.5) _projectilemovement)
    (else
      (Variables|/Script/Engine.ProjectileMovementComponent|SetProjectileGravityScale (+ _projectilegravityscale (Variables|Default|GetZeroGravPrj)) _projectilemovement))))

(fn ToyExit
  (params (Projectile :type Physics%20Ball%20Master%20Object%20Reference))
  (bind _projectilemovement (Variables|/Game/Building/ActorBlueprints/Prop/PhysicsBall_Master.PhysicsBall_Master_C|GetProjectileMovement Projectile))
  (bind _projectilegravityscale (Variables|/Script/Engine.ProjectileMovementComponent|GetProjectileGravityScale _projectilemovement))
  (if (== _projectilegravityscale (Variables|Default|GetZeroGravPrj))
    (Variables|/Script/Engine.ProjectileMovementComponent|SetProjectileGravityScale (- _projectilegravityscale (Variables|Default|GetZeroGravPrj)) _projectilemovement)
    (else
      (Variables|/Script/Engine.ProjectileMovementComponent|SetProjectileGravityScale (* _projectilegravityscale -2.0) _projectilemovement))))

(fn ConstructionScript
  (|CallParentFunction))
