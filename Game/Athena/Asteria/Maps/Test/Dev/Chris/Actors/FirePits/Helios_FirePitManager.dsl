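; Dev test manager actor that, when enabled, fires Helios fire pits through a list of timed patterns.
; Each pattern fires bursts either sequentially by index or randomly by shuffling and draining a cached pit array, via self-re-calling events with clamped delays plus random jitter.
; Finished patterns advance to the next after a loop delay, restarting or printing completion after the last.
; A separate event resets and fires every pit in a given pattern.
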(event Custom|TryNextSequencialSequence
  (bind _currentpatterndata (Variables|Default|GetCurrentPatternData))
  (bind (_patterntype_2_cc09dfd4438cba56c24e00872410ed91 _allfirepitsinpattern_9_27d3b29341c94313967c10a3b2537734 _numperburst_13_b036009140fed3484aff569200297eb6 _firedelay_5_12f4ca544e1bb6c0a9fd6398e9872bbe _randomfiredelaymax_17_f02d5cf945b5720ccd75129047a2a45a) (Utilities|Struct|BreakFirePitsDataStruct _currentpatterndata))
  ;[1 1 1] Sequencial
  (Utilities|FlowControl|Delay (Math|Float|Clamp(Float) _firedelay_5_12f4ca544e1bb6c0a9fd6398e9872bbe 0.5 9999.0))
  (Utilities|FlowControl|Delay (Math|Random|RandomFloatinRange 0.0 _randomfiredelaymax_17_f02d5cf945b5720ccd75129047a2a45a))
  (bind _finished (CallFunction|SequencialFiring))
  (if _finished
    (CallFunction|TryNextSequencialSequence)
    (else
      (bind _result (Math|Integer|IncrementInt (Variables|Default|GetCurrentPatternDataIndex))
        (:%20%20
          (if (> _result (Variables|Default|GetPatternDataLength))
            (if (Variables|Default|GetLoopWhenCompleted)
              (CallFunction|LoopAllPatterns)
              (else
                (Development|PrintString "All Patterns Completed" true true "(R=0.000000,G=0.660000,B=1.000000,A=1.000000)" 1e+01)))
            (else
              (Variables|Default|SetCurrentFirePitIndex 0)
              (CallFunction|TryNextPattern))))))))

(event EventBeginPlay
  (bind (_patterntype_2_cc09dfd4438cba56c24e00872410ed91 _allfirepitsinpattern_9_27d3b29341c94313967c10a3b2537734) (Utilities|Struct|BreakFirePitsDataStruct (Variables|Default|GetCurrentPatternData)))
  ;[1 1 1] Init
  (if (Variables|Default|GetEnabled)
    (Variables|Default|SetPatternDataLength (Utilities|Array|Length (Variables|Default|GetAllFirePitPatterns)))
    (Utilities|FlowControl|Delay (Variables|Default|GetLoopDelay))
    (if (Utilities|Array|IsValidIndex (Variables|Default|GetAllFirePitPatterns) (Variables|Default|GetCurrentPatternDataIndex))
      (bind _output_get (Variables|Default|SetCurrentPatternData (Utilities|Array|Get(acopy) (Variables|Default|GetAllFirePitPatterns) (Variables|Default|GetCurrentPatternDataIndex))))
      (switch Utilities|FlowControl|Switch|SwitchonFirePitsPatterns (Utilities|Struct|BreakFirePitsDataStruct _output_get)
        (:NewEnumerator0
          ;[1 1 1] Sequencial
          (bind _finished (CallFunction|SequencialFiring))
          (if _finished
            (CallFunction|TryNextSequencialSequence)
            (else
              (bind _result (Math|Integer|IncrementInt (Variables|Default|GetCurrentPatternDataIndex))
                (:%20%20
                  (if (> _result (Variables|Default|GetPatternDataLength))
                    (if (Variables|Default|GetLoopWhenCompleted)
                      (CallFunction|LoopAllPatterns)
                      (else
                        (Development|PrintString "All Patterns Completed" true true "(R=0.000000,G=0.660000,B=1.000000,A=1.000000)" 1e+01)))
                    (else
                      (Variables|Default|SetCurrentFirePitIndex 0)
                      (CallFunction|TryNextPattern))))))))
        (:NewEnumerator3
          ;[1 1 1] Random
          (Variables|Default|SetCachedFirePitsForRandomFire _allfirepitsinpattern_9_27d3b29341c94313967c10a3b2537734)
          (bind _finished_2 (CallFunction|RandomFiring))
          (if _finished_2
            (bind _result_2 (Math|Integer|IncrementInt (Variables|Default|GetCurrentPatternDataIndex))
              (:%20%20
                (if (> _result_2 (Variables|Default|GetPatternDataLength))
                  (if (Variables|Default|GetLoopWhenCompleted)
                    (CallFunction|LoopAllPatterns)
                    (else
                      (Development|PrintString "All Patterns Completed" true true "(R=0.000000,G=0.660000,B=1.000000,A=1.000000)" 1e+01)))
                  (else
                    (Variables|Default|SetCurrentFirePitIndex 0)
                    (Utilities|Array|Clear (Variables|Default|GetCachedFirePitsForRandomFire))
                    (CallFunction|TryNextPattern)))))
            (else
              (CallFunction|TryNextRandomFire))))))))

(event Custom|TryNextPattern
  (bind (_patterntype_2_cc09dfd4438cba56c24e00872410ed91 _allfirepitsinpattern_9_27d3b29341c94313967c10a3b2537734) (Utilities|Struct|BreakFirePitsDataStruct (Variables|Default|GetCurrentPatternData)))
  ;[1 1 1] Init
  (Utilities|FlowControl|Delay (Variables|Default|GetLoopDelay))
  (if (Utilities|Array|IsValidIndex (Variables|Default|GetAllFirePitPatterns) (Variables|Default|GetCurrentPatternDataIndex))
    (bind _output_get (Variables|Default|SetCurrentPatternData (Utilities|Array|Get(acopy) (Variables|Default|GetAllFirePitPatterns) (Variables|Default|GetCurrentPatternDataIndex))))
    (switch Utilities|FlowControl|Switch|SwitchonFirePitsPatterns (Utilities|Struct|BreakFirePitsDataStruct _output_get)
      (:NewEnumerator0
        ;[1 1 1] Sequencial
        (bind _finished (CallFunction|SequencialFiring))
        (if _finished
          (CallFunction|TryNextSequencialSequence)
          (else
            (bind _result (Math|Integer|IncrementInt (Variables|Default|GetCurrentPatternDataIndex))
              (:%20%20
                (if (> _result (Variables|Default|GetPatternDataLength))
                  (if (Variables|Default|GetLoopWhenCompleted)
                    (CallFunction|LoopAllPatterns)
                    (else
                      (Development|PrintString "All Patterns Completed" true true "(R=0.000000,G=0.660000,B=1.000000,A=1.000000)" 1e+01)))
                  (else
                    (Variables|Default|SetCurrentFirePitIndex 0)
                    (CallFunction|TryNextPattern))))))))
      (:NewEnumerator3
        ;[1 1 1] Random
        (Variables|Default|SetCachedFirePitsForRandomFire _allfirepitsinpattern_9_27d3b29341c94313967c10a3b2537734)
        (bind _finished_2 (CallFunction|RandomFiring))
        (if _finished_2
          (bind _result_2 (Math|Integer|IncrementInt (Variables|Default|GetCurrentPatternDataIndex))
            (:%20%20
              (if (> _result_2 (Variables|Default|GetPatternDataLength))
                (if (Variables|Default|GetLoopWhenCompleted)
                  (CallFunction|LoopAllPatterns)
                  (else
                    (Development|PrintString "All Patterns Completed" true true "(R=0.000000,G=0.660000,B=1.000000,A=1.000000)" 1e+01)))
                (else
                  (Variables|Default|SetCurrentFirePitIndex 0)
                  (Utilities|Array|Clear (Variables|Default|GetCachedFirePitsForRandomFire))
                  (CallFunction|TryNextPattern)))))
          (else
            (CallFunction|TryNextRandomFire)))))))

(event Custom|LoopAllPatterns
  (bind (_patterntype_2_cc09dfd4438cba56c24e00872410ed91 _allfirepitsinpattern_9_27d3b29341c94313967c10a3b2537734) (Utilities|Struct|BreakFirePitsDataStruct (Variables|Default|GetCurrentPatternData)))
  ;[1 1 1] Init
  (Variables|Default|SetCurrentPatternDataIndex 0)
  (Utilities|FlowControl|Delay (Variables|Default|GetLoopDelay))
  (if (Utilities|Array|IsValidIndex (Variables|Default|GetAllFirePitPatterns) (Variables|Default|GetCurrentPatternDataIndex))
    (bind _output_get (Variables|Default|SetCurrentPatternData (Utilities|Array|Get(acopy) (Variables|Default|GetAllFirePitPatterns) (Variables|Default|GetCurrentPatternDataIndex))))
    (switch Utilities|FlowControl|Switch|SwitchonFirePitsPatterns (Utilities|Struct|BreakFirePitsDataStruct _output_get)
      (:NewEnumerator0
        ;[1 1 1] Sequencial
        (bind _finished (CallFunction|SequencialFiring))
        (if _finished
          (CallFunction|TryNextSequencialSequence)
          (else
            (bind _result (Math|Integer|IncrementInt (Variables|Default|GetCurrentPatternDataIndex))
              (:%20%20
                (if (> _result (Variables|Default|GetPatternDataLength))
                  (if (Variables|Default|GetLoopWhenCompleted)
                    (CallFunction|LoopAllPatterns)
                    (else
                      (Development|PrintString "All Patterns Completed" true true "(R=0.000000,G=0.660000,B=1.000000,A=1.000000)" 1e+01)))
                  (else
                    (Variables|Default|SetCurrentFirePitIndex 0)
                    (CallFunction|TryNextPattern))))))))
      (:NewEnumerator3
        ;[1 1 1] Random
        (Variables|Default|SetCachedFirePitsForRandomFire _allfirepitsinpattern_9_27d3b29341c94313967c10a3b2537734)
        (bind _finished_2 (CallFunction|RandomFiring))
        (if _finished_2
          (bind _result_2 (Math|Integer|IncrementInt (Variables|Default|GetCurrentPatternDataIndex))
            (:%20%20
              (if (> _result_2 (Variables|Default|GetPatternDataLength))
                (if (Variables|Default|GetLoopWhenCompleted)
                  (CallFunction|LoopAllPatterns)
                  (else
                    (Development|PrintString "All Patterns Completed" true true "(R=0.000000,G=0.660000,B=1.000000,A=1.000000)" 1e+01)))
                (else
                  (Variables|Default|SetCurrentFirePitIndex 0)
                  (Utilities|Array|Clear (Variables|Default|GetCachedFirePitsForRandomFire))
                  (CallFunction|TryNextPattern)))))
          (else
            (CallFunction|TryNextRandomFire)))))))

(event Custom|FireAllAtPatternIndex
  (params (PatternIndex :type Integer))
  (bind (_patterntype_2_cc09dfd4438cba56c24e00872410ed91 _allfirepitsinpattern_9_27d3b29341c94313967c10a3b2537734) (Utilities|Struct|BreakFirePitsDataStruct (Utilities|Array|Get(acopy) (Variables|Default|GetAllFirePitPatterns) PatternIndex)))
  (bind _allfirepitpatterns (Variables|Default|GetAllFirePitPatterns))
  (if (Utilities|Array|IsValidIndex _allfirepitpatterns PatternIndex)
    (for _array_element _allfirepitsinpattern_9_27d3b29341c94313967c10a3b2537734
      (Class|HeliosFirePit|ResetFirePitstheFireAll _array_element))))

(event Custom|TryNextRandomFire
  (bind _currentpatterndata (Variables|Default|GetCurrentPatternData))
  (bind (_patterntype_2_cc09dfd4438cba56c24e00872410ed91 _allfirepitsinpattern_9_27d3b29341c94313967c10a3b2537734 _numperburst_13_b036009140fed3484aff569200297eb6 _firedelay_5_12f4ca544e1bb6c0a9fd6398e9872bbe _randomfiredelaymax_17_f02d5cf945b5720ccd75129047a2a45a) (Utilities|Struct|BreakFirePitsDataStruct _currentpatterndata))
  ;[1 1 1] Random
  (Utilities|FlowControl|Delay (Math|Float|Clamp(Float) _firedelay_5_12f4ca544e1bb6c0a9fd6398e9872bbe 0.5 9999.0))
  (Utilities|FlowControl|Delay (Math|Random|RandomFloatinRange 0.0 _randomfiredelaymax_17_f02d5cf945b5720ccd75129047a2a45a))
  (bind _finished (CallFunction|RandomFiring))
  (if _finished
    (bind _result (Math|Integer|IncrementInt (Variables|Default|GetCurrentPatternDataIndex))
      (:%20%20
        (if (> _result (Variables|Default|GetPatternDataLength))
          (if (Variables|Default|GetLoopWhenCompleted)
            (CallFunction|LoopAllPatterns)
            (else
              (Development|PrintString "All Patterns Completed" true true "(R=0.000000,G=0.660000,B=1.000000,A=1.000000)" 1e+01)))
          (else
            (Variables|Default|SetCurrentFirePitIndex 0)
            (Utilities|Array|Clear (Variables|Default|GetCachedFirePitsForRandomFire))
            (CallFunction|TryNextPattern)))))
    (else
      (CallFunction|TryNextRandomFire))))

(fn RandomFiring
  (bind (_patterntype_2_cc09dfd4438cba56c24e00872410ed91 _allfirepitsinpattern_9_27d3b29341c94313967c10a3b2537734 _numperburst_13_b036009140fed3484aff569200297eb6) (Utilities|Struct|BreakFirePitsDataStruct (Variables|Default|GetCurrentPatternData)))
  (bind _cachedfirepitsforrandomfire (Variables|Default|GetCachedFirePitsForRandomFire))
  (bind _returnvalue (Utilities|Array|LastIndex _cachedfirepitsforrandomfire))
  (bind _output (Utilities|Array|Get(acopy) _cachedfirepitsforrandomfire _returnvalue))
  (for _ (range 1 (+ _numperburst_13_b036009140fed3484aff569200297eb6 1))
    (Utilities|Array|Shuffle (Variables|Default|GetCachedFirePitsForRandomFire))
    (if (> (Utilities|Array|Length _cachedfirepitsforrandomfire) 1)
      (Utilities|IsValid _output
        (:Is%20Valid
          (Class|HeliosFirePit|FiretheFirePit _output)
          (Utilities|Array|RemoveIndex (Variables|Default|GetCachedFirePitsForRandomFire) _returnvalue))
        (:Is%20Not%20Valid
          _)))
    (return true))
  (return (<= (Utilities|Array|Length (Variables|Default|GetCachedFirePitsForRandomFire)) 0)))

(fn SequencialFiring
  (bind _currentpatterndata (Variables|Default|GetCurrentPatternData))
  (bind (_patterntype_2_cc09dfd4438cba56c24e00872410ed91 _allfirepitsinpattern_9_27d3b29341c94313967c10a3b2537734 _numperburst_13_b036009140fed3484aff569200297eb6) (Utilities|Struct|BreakFirePitsDataStruct _currentpatterndata))
  (bind _currentfirepitindex (Variables|Default|GetCurrentFirePitIndex))
  (bind _output (Utilities|Array|Get(acopy) _allfirepitsinpattern_9_27d3b29341c94313967c10a3b2537734 _currentfirepitindex))
  (for _ (range 1 (+ _numperburst_13_b036009140fed3484aff569200297eb6 1))
    (Utilities|IsValid _output
      (:Is%20Valid
        (Class|HeliosFirePit|FiretheFirePit _output)
        (Math|Integer|IncrementInt (Variables|Default|GetCurrentFirePitIndex)))))
  (return (> (Utilities|Array|Length _allfirepitsinpattern_9_27d3b29341c94313967c10a3b2537734) (Variables|Default|GetCurrentFirePitIndex))))

(fn ConstructionScript)
