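# Test Default Lit material, in a developer fire pit actors folder, that builds a pulsing red button glow color.
# A sine of Time blends a dark red with an overbright version of it, using the raw signed wave as alpha.
# The scalar parameter "pulse" blends from steady dim red to that animation, fully pulsing by default.
# The file shows no link from the result to a material output.
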Material /Game/Athena/Asteria/Maps/Test/Dev/Chris/Actors/FirePits/M_ButtonGlowTest
graph {
 properties {pixelDepthOffsetMode=PDOM_Legacy, usedShadingModels=Default Lit}
 nodes [
  Constant() {r=0.9}
  n1 = Constant3Vector() {constant=<rgba 0.2031 0 0 0>}
  n2 = Multiply(A=@n1, B=Constant() {r=0.1})
  n3 = Constant() {r=5}
  n4 = Time()
  LinearInterpolate(
   A=@n2
   B=LinearInterpolate(A=@n2, B=Multiply(A=@n1, B=@n3), Alpha=Sine(None=@n4))
   Alpha=ScalarParameter() {defaultValue=1, parameterName=pulse}
  )
 ]
}
